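; GA_HighTower_Soy_Boarding_Cancel_CoreBR is a Fortnite gameplay ability that cancels the HighTower Soy item's boarding state.
; On successful activation it calls the parent, then unblocks abilities tagged with the change-equipment tag on the avatar's Fort ability system.
; It then removes gameplay effects with the Boarding cooldown asset tag, which, per the inline note, starts the actual cooldown, and prints a debug string.
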(event EventRanAbilitySuccesfully
  (|CallParentFunction)
  (Tags|UnBlockAbilitiesWithTags (Ability|GetFortAbilitySystemComponent (Ability|GetAvatarActorfromActorInfo)) (Variables|Default|GetT%20ChangeEquipment) false)
  (Ability|RemoveGameplayEffectFromOwnerWithAssetTags "(GameplayTags=((TagName=\"Asset.Cooldown.HighTower.Soy.Boarding\")))") ; Removing this startst he actual cooldown
  (Development|PrintString "Cancel was succesfully used"))
