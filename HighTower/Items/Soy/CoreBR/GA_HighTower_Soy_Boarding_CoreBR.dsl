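; Gameplay ability for boarding the HighTower Soy board in CoreBR, layering ability blocking and early-exit cleanup onto its parent ability.
; InitAbility blocks abilities by tag, and the cleanup event unblocks them.
; Activation binds weapon-unequip and inventory-overflow drop events, both leading to CleanupAbility.
; Because the drop event fires only on the server, paired OnlyClientWait net syncs carry cleanup to the client, fixing boarding while a pickup is incoming.
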(event Custom|BoardDroppedMidGlide
  (params (ItemDropped :type Fort%20World%20Item%20Object%20Reference))
  (bind _returnvalue (Ability|GetAvatarActorfromActorInfo))
  (bind _returnvalue_2 (Fortnite|GetFortPlayerControllerfromActor _returnvalue))
  (bind _asfort_player_controller_athena (Utilities|Casting|CastToFortPlayerControllerAthena _returnvalue_2))
  ;[1 1 1] Fix for a bug that could happen if the board is activated while an item is about to come into your inventory.  Event only fires on the server, need to do this to make sure the client gets it and cleans up correctly
  (Ability|Tasks|WaitNetSync "OnlyClientWait"
    (:then
      (EventDispatchers|BindEventtoOnItemDroppedDuetoOverflow _asfort_player_controller_athena (delegate Custom|BoardDroppedMidGlide))
      (|CallParentFunction))
    (:OnSync
      (EventDispatchers|UnbindEventfromOnItemDroppedDuetoOverflow _asfort_player_controller_athena (delegate Custom|BoardDroppedMidGlide))
      (CallFunction|CleanupAbility))))

(event Ability|EventActivateAbility
  (bind _returnvalue (Ability|GetAvatarActorfromActorInfo))
  (bind _returnvalue_2 (Fortnite|GetFortPlayerControllerfromActor _returnvalue))
  (bind _asfort_player_controller_athena (Utilities|Casting|CastToFortPlayerControllerAthena _returnvalue_2))
  (FortPawn|BindEventtoOnWeaponUnEquippedDelegate (Utilities|Casting|CastToFortPlayerPawnAthena (Ability|GetAvatarActorfromActorInfo)) (delegate Custom|WeaponUnequipped))
  ;[1 1 1] Fix for a bug that could happen if the board is activated while an item is about to come into your inventory.  Event only fires on the server, need to do this to make sure the client gets it and cleans up correctly
  (Ability|IsLocallyControlled
    (:True
      (Ability|Tasks|WaitNetSync "OnlyClientWait"
        (:then
          _)
        (:OnSync
          (EventDispatchers|UnbindEventfromOnItemDroppedDuetoOverflow _asfort_player_controller_athena (delegate Custom|BoardDroppedMidGlide))
          (CallFunction|CleanupAbility))))
    (:False
      _))
  (EventDispatchers|BindEventtoOnItemDroppedDuetoOverflow _asfort_player_controller_athena (delegate Custom|BoardDroppedMidGlide))
  (|CallParentFunction))

(event Custom|WeaponUnequipped
  (CallFunction|CleanupAbility))

(event EventCleanupAbility
  (bind _playerpawn (Variables|Default|GetPlayerPawn))
  (FortPawn|UnbindEventfromOnWeaponUnEquippedDelegate _playerpawn (delegate Custom|WeaponUnequipped))
  (Tags|UnBlockAbilitiesWithTags (Ability|GetFortAbilitySystemComponent _playerpawn) (Variables|Default|GetBlockedTags) false)
  (Development|PrintString "Cleanup Ability Event Ran" false false)
  (|CallParentFunction))

(fn InitAbility
  (Tags|BlockAbilitiesWithTags (Ability|GetFortAbilitySystemComponent (Ability|GetAvatarActorfromActorInfo)) (Variables|Default|GetBlockedTags)) ; Unblocked after an ability finishes succesfully.
  (|CallParentFunction))

(fn OnRep_ItemDropped)
