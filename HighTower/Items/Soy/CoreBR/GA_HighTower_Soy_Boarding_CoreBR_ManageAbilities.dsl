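; Gameplay ability for the HighTower Soy Boarding item in CoreBR that turns primary presses into start or force-end events for the boarding power.
; On activation, which is always allowed, it waits for Event.Primary.Pressed and checks the avatar pawn for the Boarding.Active tag.
; If the tag is present it sends the ForceEnd.Core event.
; Otherwise it sends the Boarding.Core event, but only where the ability is not locally controlled.
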(event Ability|EventActivateAbility
  (Ability|Tasks|WaitGameplayEvent "(TagName=\"Event.Primary.Pressed\")"
    (:EventReceived
      (bind _asfort_player_pawn (Utilities|Casting|CastToFortPlayerPawn (Ability|GetAvatarActorfromActorInfo)))
      (if (GameplayTags|HasMatchingGameplayTag _asfort_player_pawn "(TagName=\"Granted.HighTower.Power.Soy.Boarding.Active\")")
        (Ability|SendGameplayEvent "(TagName=\"Event.HighTower.Power.Soy.Boarding.ForceEnd.Core\")")
        (else
          (Ability|IsLocallyControlled
            (:False
              (Ability|SendGameplayEvent "(TagName=\"Event.Hightower.Power.Soy.Boarding.Core\")"))))))))

(fn CanActivateAbility
  (params (ActorInfo :type Gameplay%20Ability%20Actor%20Info%20Structure) (Handle :type Gameplay%20Ability%20Spec%20Handle%20Structure))
  (return 0 true))
